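# RenderTargetTexturePainter is a Niagara behavior-example template that, by its name, demonstrates painting into a render target texture.
# Its single GPU emitter, PaintGrid, creates Grid2D Collection and Render Target 2D data interfaces on emitter spawn.
# A local InitializeGridToRenderTargetSize module takes both interfaces to match the grid's resolution to the render target.
# Emitter update runs only EmitterState, a sprite renderer is attached, and no paint or write stage appears here.
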NiagaraSystem /Niagara/DefaultAssets/Templates/BehaviorExamples/RenderTargetTexturePainter
dependencies {
 usedDataInterfaces [/Script/Niagara.NiagaraDataInterfaceGrid2DCollection, /Script/Niagara.NiagaraDataInterfaceRenderTarget2D]
 usedDynamicInputs []
 usedModules [&SystemState]
}
emitters {PaintGrid {
  simTarget=GPUComputeSim
  emitterSpawnScript [
   /SetVariables(
    Emitter.Grid2D Collection=DataInterface() {propertyValues {
      ClearBeforeNonIterationStage=true
      NumAttributes=4
      NumCellsMaxAxis=3
      NumCellsX=3
      NumCellsY=3
      OverrideBufferFormat=Float
      PreviewAttribute=None
      SetGridFromMaxAxis=false
      bOverrideFormat=false
      bPreviewGrid=false
      RenderTargetUserParameter {parameter {name=None, typeDefHandle {registeredTypeIndex=4}, varData []}}
      WorldBBoxSize=<100 100>
     }}
    Emitter.Render Target 2D=DataInterface() {propertyValues {
      MipMapGeneration=Disabled
      MipMapGenerationType=Linear
      OverrideRenderTargetFilter=TF_Default
      OverrideRenderTargetFormat=RTF_R8
      bInheritUserParameterSettings=false
      bOverrideFormat=false
      bPreviewRenderTarget=false
      RenderTargetUserParameter {parameter {name=None, typeDefHandle {registeredTypeIndex=4}, varData []}}
      Size=<512 512>
     }}
   ) {isSetParameters=true}
   &self:InitializeGridToRenderTargetSize(
    Grid2D Collection:NiagaraDataInterfaceGrid2DCollection=$Emitter.Grid2D Collection
    Render Target 2D:NiagaraDataInterfaceRenderTarget2D=$Emitter.Render Target 2D
   ) {module=InitializeGridToRenderTargetSize}
  ]
  emitterUpdateScript [&EmitterState()]
  rendererClasses [&NiagaraSpriteRendererProperties]
  renderers [{rendererClass=&NiagaraSpriteRendererProperties}]
 }}
summary {
 systemName=RenderTargetTexturePainter
 emitters [{emitterName=PaintGrid, simTarget=GPUComputeSim, rendererClasses [&NiagaraSpriteRendererProperties]}]
 userVariables []
}
